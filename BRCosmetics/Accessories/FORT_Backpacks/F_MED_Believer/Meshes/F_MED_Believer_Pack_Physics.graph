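# Physics asset for the F_MED_Believer backpack cosmetic, defining bodies and joints for its dangling buttons, handle, tassel and zipper.
# Each part has a small base sphere constrained to a parent body shaped as a sphere, capsule or box.
# The tassel adds a second capsule link to form a short chain.
# All joints keep twist tight, with narrower swing on buttons and wider or asymmetric swing on the other parts.
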PhysicsAsset /BRCosmetics/Accessories/FORT_Backpacks/F_MED_Believer/Meshes/F_MED_Believer_Pack_Physics
bodies {
 dyn_button_base_01 [{shapeType=Sphere, radius=0.5}]
 dyn_button_base_02 [{shapeType=Sphere, radius=0.5}]
 dyn_button_base_03 [{shapeType=Sphere, radius=0.5}]
 dyn_button_base_04 [{shapeType=Sphere, radius=0.5}]
 dyn_button_base_05 [{shapeType=Sphere, radius=0.5}]
 dyn_button_parent_01 [{shapeType=Sphere, radius=0.7051, center=<0.3042 0 0>}]
 dyn_button_parent_02 [{shapeType=Sphere, radius=1.0261, center=<0.3761 0 0>}]
 dyn_button_parent_03 [{shapeType=Sphere, radius=1.0209, center=<0.6091 0 0>}]
 dyn_button_parent_04 [{shapeType=Sphere, radius=1.0207, center=<0.3432 0 0>}]
 dyn_button_parent_05 [{shapeType=Sphere, radius=0.7723, center=<0.3907 0 0>}]
 dyn_handle_base [{shapeType=Sphere, radius=0.5}]
 dyn_handle_parent [{shapeType=Capsule, length=1.1905, radius=0.9538, center=<-0.5769 0 0>, rotation=<rot 0 0 90>}]
 dyn_tassel_01 [{shapeType=Capsule, length=0.776, radius=0.7738, center=<1.0325 0 0>, rotation=<rot 90 0 0>}]
 dyn_tassel_base [{shapeType=Sphere, radius=0.5}]
 dyn_tassel_parent [{shapeType=Capsule, length=0.6906, radius=0.7293, center=<0.9908 0 0>, rotation=<rot 90 0 0>}]
 dyn_zipper_base [{shapeType=Sphere, radius=0.5}]
 dyn_zipper_parent [{shapeType=Box, extentX=3.0633, extentY=1.5929, extentZ=0.5338, center=<1.5525 0 0>}]
}
constraints [
 {bone1Name=dyn_tassel_parent, bone2Name=dyn_tassel_base, swing1LimitDegrees=30, swing2LimitDegrees=20, twistLimitDegrees=3}
 {bone1Name=dyn_tassel_01, bone2Name=dyn_tassel_parent, swing1LimitDegrees=25, swing2LimitDegrees=15, twistLimitDegrees=3}
 {bone1Name=dyn_handle_parent, bone2Name=dyn_handle_base, swing1LimitDegrees=30, swing2LimitDegrees=30, twistLimitDegrees=3}
 {bone1Name=dyn_button_parent_05, bone2Name=dyn_button_base_05, swing1LimitDegrees=15, swing2LimitDegrees=15, twistLimitDegrees=3}
 {bone1Name=dyn_button_parent_04, bone2Name=dyn_button_base_04, swing1LimitDegrees=10, swing2LimitDegrees=10, twistLimitDegrees=3}
 {bone1Name=dyn_button_parent_03, bone2Name=dyn_button_base_03, swing1LimitDegrees=15, swing2LimitDegrees=15, twistLimitDegrees=3}
 {bone1Name=dyn_button_parent_02, bone2Name=dyn_button_base_02, swing1LimitDegrees=15, swing2LimitDegrees=15, twistLimitDegrees=3}
 {bone1Name=dyn_button_parent_01, bone2Name=dyn_button_base_01, swing1LimitDegrees=15, swing2LimitDegrees=15, twistLimitDegrees=3}
 {bone1Name=dyn_zipper_parent, bone2Name=dyn_zipper_base, swing1LimitDegrees=30, swing2LimitDegrees=30, twistLimitDegrees=3}
]
